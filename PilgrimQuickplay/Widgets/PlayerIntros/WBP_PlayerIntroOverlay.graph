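# Pilgrim Quickplay intro overlay that briefly shows a player's uppercase name and instrument icon with animated pulse glows.
# A stateful PlayerChanged event calls SetupAndShow, which, gated by a console variable, sets name, song-aware icon and identity color on dynamic materials.
# The identity pill shows only for local players in local multiplayer.
# Playback always uses the Front animation; the squad-index selection of positional animations is left unconnected.
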WidgetBlueprint /PilgrimQuickplay/Widgets/PlayerIntros/WBP_PlayerIntroOverlay
animations [
 A_Position01_Front {
  binding_count=7
  length_seconds=2
  object=&self:A_Position01_Front
  sequence {bindings [
    PlayerName_Overlay {
     id=#1
     possessed_class=Overlay
     tracks [{
       type=MovieScene2DTransformTrack
       sections [{
         type=MovieScene2DTransformSection
         channels [
          Translation.X {type=MovieSceneScriptingFloatChannel}
          Translation.Y {type=MovieSceneScriptingFloatChannel, keys [[0,-20],[5,0],[35,10],[40,20]]}
         ]
        }]
      }]
    }
    Instrument_Overlay {
     id=#2
     possessed_class=Overlay
     tracks [{
       type=MovieScene2DTransformTrack
       sections [{
         type=MovieScene2DTransformSection
         channels [
          Translation.X {type=MovieSceneScriptingFloatChannel}
          Translation.Y {type=MovieSceneScriptingFloatChannel, keys [[0,0],[5,-20],[35,-30],[40,-40]]}
         ]
        }]
      }]
    }
    Pulse_Glow_Echo {
     id=#3
     possessed_class=Image
     tracks [
      {
       type=MovieSceneWidgetMaterialTrack
       sections [{type=MovieSceneParameterSection, channels [TaperGlow {type=MovieSceneScriptingFloatChannel, keys [[6,15],[19,20],[40,5]]}]}]
      }
      {
       type=MovieScene2DTransformTrack
       sections [{
         type=MovieScene2DTransformSection
         channels [
          Translation.X {type=MovieSceneScriptingFloatChannel}
          Translation.Y {type=MovieSceneScriptingFloatChannel}
          Angle {type=MovieSceneScriptingFloatChannel}
          Scale.X {type=MovieSceneScriptingFloatChannel, keys [[6,1.2],[19,2.25]]}
          Scale.Y {type=MovieSceneScriptingFloatChannel, keys [[6,1.2],[19,2.25]]}
          Shear.X {type=MovieSceneScriptingFloatChannel}
          Shear.Y {type=MovieSceneScriptingFloatChannel}
         ]
        }]
      }
      {
       type=MovieSceneFloatTrack
       display_name=Render Opacity
       sections [{
         type=MovieSceneFloatSection
         end_frame=25
         start_frame=0
         channels [None {type=MovieSceneScriptingFloatChannel, keys [[6,0],[9,1],[14,1],[19,0]]}]
        }]
      }
     ]
    }
    InstrumentImage {
     id=#4
     possessed_class=Image
     tracks [
      {
       type=MovieScene2DTransformTrack
       sections [{
         type=MovieScene2DTransformSection
         channels [
          Scale.X {type=MovieSceneScriptingFloatChannel, keys [[0,0.5],[3,0.5],[9,1],[35,1],[40,0.8]]}
          Scale.Y {type=MovieSceneScriptingFloatChannel, keys [[0,0.5],[3,0.5],[9,1],[35,1],[40,0.8]]}
         ]
        }]
      }
      {
       type=MovieSceneFloatTrack
       display_name=Render Opacity
       sections [{type=MovieSceneFloatSection, channels [None {type=MovieSceneScriptingFloatChannel, keys [[0,0],[3,0],[6,1],[35,1],[40,0]]}]}]
      }
     ]
    }
    Pulse_Glow {
     id=#5
     possessed_class=Image
     tracks [
      {
       type=MovieSceneWidgetMaterialTrack
       sections [{
         type=MovieSceneParameterSection
         channels [
          RingSize {type=MovieSceneScriptingFloatChannel, keys [[0,0],[11,0.9]]}
          ProgressThickness {type=MovieSceneScriptingFloatChannel, keys [[0,0.3],[3,0.4],[11,0.2]]}
          TaperGlow {type=MovieSceneScriptingFloatChannel, keys [[0,10],[6,20]]}
         ]
        }]
      }
      {
       type=MovieSceneFloatTrack
       display_name=Render Opacity
       sections [{
         type=MovieSceneFloatSection
         end_frame=25
         start_frame=0
         channels [None {type=MovieSceneScriptingFloatChannel, keys [[0,0],[3,1],[6,1],[11,0]]}]
        }]
      }
     ]
    }
    IntroOverlay {
     id=#6
     possessed_class=Overlay
     tracks [{
       type=MovieSceneFloatTrack
       display_name=Render Opacity
       sections [{type=MovieSceneFloatSection, channels [None {type=MovieSceneScriptingFloatChannel, keys [[0,0],[5,1],[35,1],[40,0]]}]}]
      }]
    }
    PlayerIdentity_Pill {
     id=#7
     possessed_class=WBP_Pilgrim_PlayerIdentity_Pill_C
     tracks [
      {
       type=MovieScene2DTransformTrack
       sections [{
         type=MovieScene2DTransformSection
         channels [
          Translation.X {type=MovieSceneScriptingFloatChannel}
          Translation.Y {type=MovieSceneScriptingFloatChannel, keys [[0,-20],[5,-5],[35,5],[40,15]]}
         ]
        }]
      }
      {
       type=MovieSceneFloatTrack
       display_name=Render Opacity
       sections [{type=MovieSceneFloatSection, channels [None {type=MovieSceneScriptingFloatChannel, keys [[0,0],[5,1],[35,1],[40,0]]}]}]
      }
     ]
    }
   ]}
 }
 A_Position02_Left {
  binding_count=5
  length_seconds=2
  object=&self:A_Position02_Left
  sequence {bindings [
    Instrument_Overlay {
     id=#8
     possessed_class=Overlay
     tracks [{
       type=MovieScene2DTransformTrack
       sections [{type=MovieScene2DTransformSection, channels [Translation.X {type=MovieSceneScriptingFloatChannel, keys [[0,-20],[5,0],[35,10],[40,-20]]}]}]
      }]
    }
    PlayerName_Overlay {
     id=#9
     possessed_class=Overlay
     tracks [{
       type=MovieScene2DTransformTrack
       sections [{type=MovieScene2DTransformSection, channels [Translation.X {type=MovieSceneScriptingFloatChannel, keys [[0,20],[5,0],[35,-10],[40,20]]}]}]
      }]
    }
    IntroOverlay {
     id=#10
     possessed_class=Overlay
     tracks [{
       type=MovieSceneFloatTrack
       display_name=Render Opacity
       sections [{type=MovieSceneFloatSection, channels [None {type=MovieSceneScriptingFloatChannel, keys [[0,0],[5,1],[35,1],[40,0]]}]}]
      }]
    }
   ]}
 }
 A_Position03_Right {
  binding_count=3
  length_seconds=2
  object=&self:A_Position03_Right
  sequence {bindings [
    Instrument_Overlay {
     id=#11
     possessed_class=Overlay
     tracks [{
       type=MovieScene2DTransformTrack
       sections [{type=MovieScene2DTransformSection, channels [Translation.X {type=MovieSceneScriptingFloatChannel, keys [[0,20],[5,0],[35,-10],[40,20]]}]}]
      }]
    }
    IntroOverlay {
     id=#12
     possessed_class=Overlay
     tracks [{
       type=MovieSceneFloatTrack
       display_name=Render Opacity
       sections [{type=MovieSceneFloatSection, channels [None {type=MovieSceneScriptingFloatChannel, keys [[0,0],[5,1],[35,1],[40,0]]}]}]
      }]
    }
    PlayerName_Overlay {
     id=#13
     possessed_class=Overlay
     tracks [{
       type=MovieScene2DTransformTrack
       sections [{type=MovieScene2DTransformSection, channels [Translation.X {type=MovieSceneScriptingFloatChannel, keys [[0,-20],[5,0],[35,10],[40,-20]]}]}]
      }]
    }
   ]}
 }
 A_Position04_Back {
  binding_count=3
  length_seconds=2
  object=&self:A_Position04_Back
  sequence {bindings [
    Instrument_Overlay {
     id=#14
     possessed_class=Overlay
     tracks [{
       type=MovieScene2DTransformTrack
       sections [{
         type=MovieScene2DTransformSection
         channels [
          Translation.X {type=MovieSceneScriptingFloatChannel}
          Translation.Y {type=MovieSceneScriptingFloatChannel, keys [[0,-10],[5,10],[35,20],[40,-10]]}
         ]
        }]
      }]
    }
    IntroOverlay {
     id=#15
     possessed_class=Overlay
     tracks [{
       type=MovieSceneFloatTrack
       display_name=Render Opacity
       sections [{type=MovieSceneFloatSection, channels [None {type=MovieSceneScriptingFloatChannel, keys [[0,0],[5,1],[35,1],[40,0]]}]}]
      }]
    }
    PlayerName_Overlay {
     id=#16
     possessed_class=Overlay
     tracks [{
       type=MovieScene2DTransformTrack
       sections [{
         type=MovieScene2DTransformSection
         channels [
          Translation.X {type=MovieSceneScriptingFloatChannel}
          Translation.Y {type=MovieSceneScriptingFloatChannel, keys [[0,30],[5,10],[35,0],[40,30]]}
         ]
        }]
      }]
    }
   ]}
 }
 A_IntroFlare {
  binding_count=4
  length_seconds=2
  object=&self:A_IntroFlare
  sequence {bindings [
    Pulse_Glow {
     id=#17
     possessed_class=Image
     tracks [
      {
       type=MovieSceneWidgetMaterialTrack
       sections [{
         type=MovieSceneParameterSection
         channels [
          RingSize {type=MovieSceneScriptingFloatChannel, keys [[0,0],[11,0.9]]}
          ProgressThickness {type=MovieSceneScriptingFloatChannel, keys [[0,0.3],[3,0.4],[11,0.2]]}
          TaperGlow {type=MovieSceneScriptingFloatChannel, keys [[0,10],[6,20]]}
         ]
        }]
      }
      {
       type=MovieSceneFloatTrack
       display_name=Render Opacity
       sections [{
         type=MovieSceneFloatSection
         end_frame=25
         start_frame=0
         channels [None {type=MovieSceneScriptingFloatChannel, keys [[0,0],[3,1],[6,1],[11,0]]}]
        }]
      }
     ]
    }
    Pulse_Glow_Echo {
     id=#18
     possessed_class=Image
     tracks [
      {
       type=MovieSceneWidgetMaterialTrack
       sections [{type=MovieSceneParameterSection, channels [TaperGlow {type=MovieSceneScriptingFloatChannel, keys [[6,15],[19,20],[40,5]]}]}]
      }
      {
       type=MovieScene2DTransformTrack
       sections [{
         type=MovieScene2DTransformSection
         channels [
          Translation.X {type=MovieSceneScriptingFloatChannel}
          Translation.Y {type=MovieSceneScriptingFloatChannel}
          Angle {type=MovieSceneScriptingFloatChannel}
          Scale.X {type=MovieSceneScriptingFloatChannel, keys [[6,1.2],[19,2.25]]}
          Scale.Y {type=MovieSceneScriptingFloatChannel, keys [[6,1.2],[19,2.25]]}
          Shear.X {type=MovieSceneScriptingFloatChannel}
          Shear.Y {type=MovieSceneScriptingFloatChannel}
         ]
        }]
      }
      {
       type=MovieSceneFloatTrack
       display_name=Render Opacity
       sections [{
         type=MovieSceneFloatSection
         end_frame=25
         start_frame=0
         channels [None {type=MovieSceneScriptingFloatChannel, keys [[6,0],[9,1],[14,1],[19,0]]}]
        }]
      }
     ]
    }
    InstrumentImage {
     id=#19
     possessed_class=Image
     tracks [
      {
       type=MovieScene2DTransformTrack
       sections [{
         type=MovieScene2DTransformSection
         channels [
          Scale.X {type=MovieSceneScriptingFloatChannel, keys [[0,0.5],[3,0.5],[9,1],[35,1],[40,0.8]]}
          Scale.Y {type=MovieSceneScriptingFloatChannel, keys [[0,0.5],[3,0.5],[9,1],[35,1],[40,0.8]]}
         ]
        }]
      }
      {
       type=MovieSceneFloatTrack
       display_name=Render Opacity
       sections [{type=MovieSceneFloatSection, channels [None {type=MovieSceneScriptingFloatChannel, keys [[0,0],[3,0],[6,1],[35,1],[40,0]]}]}]
      }
     ]
    }
    PlayerName_TextBlock {
     id=#20
     possessed_class=UEFN_TextBlock_C
     tracks [
      {
       type=MovieScene2DTransformTrack
       sections [{
         type=MovieScene2DTransformSection
         end_frame=40
         start_frame=0
         channels [
          Translation.X {type=MovieSceneScriptingFloatChannel, keys [[0,-200],[3,-10],[35,10],[40,200]]}
          Translation.Y {type=MovieSceneScriptingFloatChannel, const [0,1]}
          Scale.X {type=MovieSceneScriptingFloatChannel, keys [[0,1.3],[3,1],[35,1],[40,1.3]]}
          Scale.Y {type=MovieSceneScriptingFloatChannel, const [1,1]}
         ]
        }]
      }
      {
       type=MovieSceneFloatTrack
       display_name=Render Opacity
       sections [{
         type=MovieSceneFloatSection
         end_frame=40
         start_frame=0
         channels [None {type=MovieSceneScriptingFloatChannel, keys [[0,0],[3,1],[35,1],[40,0]]}]
        }]
      }
     ]
    }
   ]}
 }
]
graphs [EventGraph {nodes [
   n1 = ForceSetupAndShow!() -> (OutputDelegate, then, Player State:Player State Object Reference) {is=title K2Node_CustomEvent}
   > n2 = Set PlayerStateForSetupAndShow(PlayerStateForSetupAndShow:Player State Object Reference=@n1.Player State) -> (Output_Get:Player State Object Reference)
   n3 = Get PlayerStateForSetupAndShow() -> (PlayerStateForSetupAndShow:Player State Object Reference)
   n4 = GetPlayerName!(self:Player State Object Reference=@n3) -> (ReturnValue:String)
   Set Text(
    execute=[@n15.then, @n2.then]
    Text="To Text (String)"(InString=ToUpper!(SourceString:String=@n4) -> (ReturnValue:String))
    self=Get PlayerName_TextBlock() -> (PlayerName_TextBlock:Text Block Object Reference)
   )
   > n5 = Sequence!(execute) -> (then_0, then_1, then_2:Exec, then_3:Exec)
   > n6 = GetPilgrimQuickplayPlayerStateComponent!(execute, PlayerState:Player State Object Reference=@n3) -> (then, ReturnValue:Pilgrim Quickplay Player State Component Object Reference)
   > n7 = Is Valid(InputObject=@n6)
   PrintString!(execute=@n7.Is Not Valid, InString=Missing Pilgrim Quickplay Player State Component, bPrintToScreen, bPrintToLog, TextColor, Duration, Key) -> (then)
   n8 = GetOwningPlayer()
   n9 = GetInstrumentChoice!(self:Pilgrim Core Player State Component Object Reference=@n6) -> (ReturnValue:EPilgrimTrackType Enum=TrackGuitar)
   SetBrushFromSoftMaterial!(
    execute=@n7.Is Valid
    self:Image Object Reference=Get InstrumentImage() -> (InstrumentImage:Image Object Reference)
    SoftMaterial:Material Interface Soft Object Reference=GetIconMaterialByInstrument!(
     Instrument:EPilgrimTrackType Enum=@n9
     SongMetaData:Pilgrim Song Metadata Object Reference=GetCurrentSong!(ContextActor:Actor Object Reference=@n8) -> (ReturnValue:Pilgrim Song Metadata Object Reference)
     SparksInstrument:ESparksInstrumentType Enum=PilgrimTrackTypeToSparksInstrument!(PilgrimType:EPilgrimTrackType Enum=@n9) -> (ReturnValue:ESparksInstrumentType Enum=Guitar)
     bUseKeytarIcon:bool=true
     UseProIcons:bool=true
    ) -> (IconMaterial:Material Instance Soft Object Reference)
   ) -> (then)
   n10 = Cast To FortPlayerStateAthena!(execute=@n5.then_1, Object=@n3) -> (then, CastFailed, AsFort Player State Athena:Fort Player State Athena Object Reference) {is=title K2Node_DynamicCast}
   > Set Fort Player State Athena(Fort Player State Athena:Fort Player State Athena Object Reference=@n10) -> (Output_Get:Fort Player State Athena Object Reference)
   n11 = GetPilgrim PlayerIdentity Color By PlayerState!(Player State:Player State Object Reference=@n10) -> (Player Identity Color:Linear Color Structure)
   > SetVectorParameterValue(self=Get PlayerBackMatInst() -> (PlayerBackMatInst:Material Instance Dynamic Object Reference), ParameterName=ColorB, Value=@n11)
   > SetVectorParameterValue(self=Get PulseGlowMatInst() -> (PulseGlowMatInst:Material Instance Dynamic Object Reference), ParameterName=Color_A, Value=@n11)
   > SetVectorParameterValue(self=Get PulseGlowMatInst(), ParameterName=Color_B, Value=@n11)
   > SetVectorParameterValue(self=Get PulseGlowEchoMatInst() -> (PulseGlowEchoMatInst:Material Instance Dynamic Object Reference), ParameterName=Color_A, Value=@n11)
   > SetVectorParameterValue(self=Get PulseGlowEchoMatInst(), ParameterName=Color_B, Value=@n11)
   PlayAnimation(execute=@n5.then_3, InAnimation=Get A_Position01_Front() -> (A_Position01_Front:Widget Animation Object Reference))
   > SetVisibility(self=Get IntroOverlay() -> (IntroOverlay:Overlay Object Reference), InVisibility=HitTestInvisible)
   n12 = Branch(execute=@n5.then_2, Condition=Is Local Player in Local MP Session!(PlayerState:Player State Object Reference=@n3) -> (ReturnValue:bool))
   > OverridePlayerInfo!(
    execute
    self:WBP Pilgrim Player Identity Pill Object Reference=Get PlayerIdentity_Pill() -> (PlayerIdentity_Pill:WBP Pilgrim Player Identity Pill Object Reference)
    PlayerState:Player State Object Reference=@n3
   ) -> (then)
   > SetVisibility(self=Get PlayerIdentity_Pill(), InVisibility=HitTestInvisible)
   SetVisibility(execute=@n12.else, self=Get PlayerIdentity_Pill(), InVisibility=Collapsed)
   n13 = SetupAndShow!() -> (OutputDelegate, then, PlayerState:Player State Object Reference) {is=title K2Node_CustomEvent}
   > Set PlayerStateForSetupAndShow(PlayerStateForSetupAndShow=@n13.PlayerState)
   > n14 = GetConsoleVariableBoolValue!(execute, VariableName:String=Pilgrim.ShowIntroPlayerOverlay) -> (then, ReturnValue:bool)
   > n15 = Branch(Condition=@n14)
   Event Destruct()
   n16 = Get IntroPlayerChangedEventHandle() -> (IntroPlayerChangedEventHandle:Async Action Start Listening To Stateful Event Object Reference)
   > Is Valid(InputObject=@n16)
   > StopListeningToStatefulEvent(self:Async Action Start Listening To Stateful Event Object Reference=@n16)
   > Set IntroPlayerChangedEventHandle(IntroPlayerChangedEventHandle:Async Action Start Listening To Stateful Event Object Reference) -> (Output_Get:Async Action Start Listening To Stateful Event Object Reference)
   Event Construct()
   > n17 = GetDynamicMaterial(self:Border Object Reference=Get PlayerBack_Border() -> (PlayerBack_Border:Border Object Reference))
   > Set PlayerBackMatInst(PlayerBackMatInst:Material Instance Dynamic Object Reference=@n17) -> (Output_Get:Material Instance Dynamic Object Reference)
   > n18 = GetDynamicMaterial(self:Image Object Reference=Get Pulse_Glow() -> (Pulse_Glow:Image Object Reference))
   > Set PulseGlowMatInst(PulseGlowMatInst:Material Instance Dynamic Object Reference=@n18) -> (Output_Get:Material Instance Dynamic Object Reference)
   > n19 = GetDynamicMaterial(self=Get Pulse_Glow_Echo() -> (Pulse_Glow_Echo:Image Object Reference))
   > Set PulseGlowEchoMatInst(PulseGlowEchoMatInst:Material Instance Dynamic Object Reference=@n19) -> (Output_Get:Material Instance Dynamic Object Reference)
   > n20 = PilgrimWaitForEventRouter!(execute:Exec, ContextActor:Actor Object Reference=GetOwningPlayer()) -> (then:Exec, OnEventRouterAvailable:Exec, EventRouter:Gameplay Event Router Component Object Reference) {is=title K2Node_AsyncAction}
   n21 = StartListeningToStatefulEvent!(
    execute:Exec=@n20.OnEventRouterAvailable
    Target:Gameplay Event Router Component Object Reference=@n20
    EventType:Script Struct Object Reference=/Script/PilgrimQuickplayRuntime.PilgrimEvent_Intro_PlayerChanged
    Context:Object Reference
    EventBubblingRule:EEventBubblingRule Enum
   ) -> (
    then:Exec
    AsyncTaskProxy:Async Action Start Listening To Stateful Event Object Reference
    OnEventReceived:Exec
    OnSavedEventStateExists:Exec
    OnEventStateCleared:Exec
    Payload:Pilgrim Event Intro Player Changed Structure
   ) {is=title K2Node_StartListeningToStatefulEventAsyncAction}
   > Set IntroPlayerChangedEventHandle(IntroPlayerChangedEventHandle=@n21)
   SetupAndShow!(execute=@n21.OnEventStateCleared, self, PlayerState:Player State Object Reference) -> (then) {is=title K2Node_CallFunction}
   SetupAndShow(
    execute=[@n21.OnEventReceived, @n21.OnSavedEventStateExists]
    PlayerState=Break Pilgrim Event Intro Player Changed!(PilgrimEvent_Intro_PlayerChanged:"Pilgrim Event Intro Player Changed Structure (by ref)"=@n21.Payload) -> (PlayerState:Player State Object Reference) {is=title K2Node_BreakStruct}
   )
   n22 = GetSquadMemberIndex!(self:Fort Player State Athena Object Reference=Get Fort Player State Athena() -> (Fort Player State Athena:Fort Player State Athena Object Reference)) -> (ReturnValue:int=0)
   Is Valid(InputObject=Select!(
     Option 0:Widget Animation Object Reference=Get A_Position01_Front()
     Option 1:Widget Animation Object Reference=Get A_Position02_Left() -> (A_Position02_Left:Widget Animation Object Reference)
     Option 2:Widget Animation Object Reference=Get A_Position03_Right() -> (A_Position03_Right:Widget Animation Object Reference)
     Option 3:Widget Animation Object Reference=Get A_Position04_Back() -> (A_Position04_Back:Widget Animation Object Reference)
     Index:int="Clamp (Integer)"(Value=@n22, Min=0, Max=3) -> (ReturnValue=0)
    ) -> (ReturnValue:Widget Animation Object Reference))
  ]}]
variables [
 Fort Player State Athena {type=FortPlayerStateAthena, default=null}
 IntroPlayerChangedEventHandle {type=AsyncAction_StartListeningToStatefulEvent, default=null}
 PlayerBackMatInst {type=MaterialInstanceDynamic}
 PlayerState {type=PlayerState, default=null}
 PlayerStateForSetupAndShow {type=PlayerState, default=null}
 PulseGlowEchoMatInst {type=MaterialInstanceDynamic}
 PulseGlowMatInst {type=MaterialInstanceDynamic}
]
widgets {
 info {inheritedWidgetCount=0, parentClass=&CommonUserWidget, rootWidgetClass=&Overlay, widgetCount=10}
 widgets [
  IntroOverlay {
   widgetClassPath=&Overlay
   bIsVariable=true
   widget=IntroOverlay
   properties {pixelSnapping=Disabled, visibility=Hidden, renderTransform {angle=0, scale=<1 1>, shear=<0 0>, translation=<0 -60>}}
  }
  Instrument_Overlay {
   widgetClassPath=&Overlay
   parent=IntroOverlay
   widget=Instrument_Overlay
   slot {class=OverlaySlot, properties {horizontalAlignment=HAlign_Center, verticalAlignment=VAlign_Bottom, padding=<ltrb 0 0 0 56>}}
  }
  Pulse_Glow_Echo {
   widgetClassPath=&Image
   bIsVariable=true
   parent=Instrument_Overlay
   widget=Pulse_Glow_Echo
   properties {
    renderOpacity=0
    visibility=SelfHitTestInvisible
    brush {
     bIsDynamicallyLoaded=false
     drawAs=Image
     imageType=NoImage
     mirroring=NoMirror
     resourceName=None
     resourceObject=/PilgrimCore/Widgets/Materials/MI_UI_PlayerIntro_Pulse
     tiling=NoTile
     imageSize=<300 300>
     margin=<ltrb 0 0 0 0>
     outlineSettings {
      bUseBrushTransparency=false
      roundingType=HalfHeightRadius
      width=0
      color {colorUseRule=UseColor_Specified, specifiedColor=<rgba 0 0 0 0>}
      cornerRadii=<0 0 0 0>
     }
     tintColor {colorUseRule=UseColor_Specified, specifiedColor=<rgba 1 1 1 1>}
     uVRegion {bIsValid=false, max=<0 0>, min=<0 0>}
    }
    renderTransform {angle=0, scale=<1.2 1.2>, shear=<0 0>, translation=<0 0>}
   }
   slot {class=OverlaySlot, properties {horizontalAlignment=HAlign_Center, verticalAlignment=VAlign_Center}}
  }
  InstrumentImage {
   widgetClassPath=&Image
   bIsVariable=true
   parent=Instrument_Overlay
   widget=InstrumentImage
   properties {
    visibility=HitTestInvisible
    brush {
     bIsDynamicallyLoaded=false
     drawAs=Image
     imageType=NoImage
     mirroring=NoMirror
     resourceName=None
     resourceObject=/PilgrimQuickplay/Widgets/Materials/MI_UI_Icon_Instrument_Type_Guitar
     tiling=NoTile
     imageSize=<320 320>
     margin=<ltrb 0 0 0 0>
     outlineSettings {
      bUseBrushTransparency=false
      roundingType=HalfHeightRadius
      width=0
      color {colorUseRule=UseColor_Specified, specifiedColor=<rgba 0 0 0 0>}
      cornerRadii=<0 0 0 0>
     }
     tintColor {colorUseRule=UseColor_Specified, specifiedColor=<rgba 1 1 1 1>}
     uVRegion {bIsValid=false, max=<0 0>, min=<0 0>}
    }
   }
   slot {class=OverlaySlot, properties {horizontalAlignment=HAlign_Center}}
  }
  Pulse_Glow {
   widgetClassPath=&Image
   bIsVariable=true
   parent=Instrument_Overlay
   widget=Pulse_Glow
   properties {
    renderOpacity=0
    visibility=SelfHitTestInvisible
    brush {
     bIsDynamicallyLoaded=false
     drawAs=Image
     imageType=NoImage
     mirroring=NoMirror
     resourceName=None
     resourceObject=&MI_UI_PlayerIntro_Pulse
     tiling=NoTile
     imageSize=<300 300>
     margin=<ltrb 0 0 0 0>
     outlineSettings {
      bUseBrushTransparency=false
      roundingType=HalfHeightRadius
      width=0
      color {colorUseRule=UseColor_Specified, specifiedColor=<rgba 0 0 0 0>}
      cornerRadii=<0 0 0 0>
     }
     tintColor {colorUseRule=UseColor_Specified, specifiedColor=<rgba 1 1 1 1>}
     uVRegion {bIsValid=false, max=<0 0>, min=<0 0>}
    }
    renderTransform {angle=0, scale=<1.2 1.2>, shear=<0 0>, translation=<0 0>}
   }
   slot {class=OverlaySlot, properties {horizontalAlignment=HAlign_Center, verticalAlignment=VAlign_Center}}
  }
  PlayerName_Overlay {
   widgetClassPath=&Overlay
   parent=IntroOverlay
   widget=PlayerName_Overlay
   slot {class=OverlaySlot, properties {horizontalAlignment=HAlign_Center, verticalAlignment=VAlign_Bottom, padding=<ltrb 0 0 0 88>}}
  }
  PlayerBack_Border {
   widgetClassPath=&Border
   bIsVariable=true
   parent=PlayerName_Overlay
   widget=PlayerBack_Border
   properties {
    horizontalAlignment=HAlign_Center
    pixelSnapping=Disabled
    verticalAlignment=VAlign_Top
    visibility=SelfHitTestInvisible
    background {
     bIsDynamicallyLoaded=false
     drawAs=Image
     imageType=NoImage
     mirroring=NoMirror
     resourceName=None
     resourceObject=/FortUILibrary/MaterialInstances/MI_UI_CylinderGradient
     tiling=NoTile
     imageSize=<32 32>
     margin=<ltrb 0 0 0 0>
     outlineSettings {
      bUseBrushTransparency=false
      roundingType=HalfHeightRadius
      width=0
      color {colorUseRule=UseColor_Specified, specifiedColor=<rgba 0 0 0 0>}
      cornerRadii=<0 0 0 0>
     }
     tintColor {colorUseRule=UseColor_Specified, specifiedColor=<rgba 1 1 1 1>}
     uVRegion {bIsValid=false, max=<0 0>, min=<0 0>}
    }
    padding=<ltrb 128 8 128 8>
   }
   slot {class=OverlaySlot, properties {horizontalAlignment=HAlign_Fill, verticalAlignment=VAlign_Fill}}
  }
  Overlay_0 {
   widgetClassPath=&Overlay
   parent=PlayerBack_Border
   widget=Overlay_0
   slot {class=BorderSlot, properties {horizontalAlignment=HAlign_Center, verticalAlignment=VAlign_Top, padding=<ltrb 128 8 128 8>}}
  }
  PlayerName_TextBlock {
   widgetClassPath=&UEFN_TextBlock
   bIsVariable=true
   parent=Overlay_0
   widget=PlayerName_TextBlock
   properties {
    text=Player Name
    visibility=HitTestInvisible
    font {
     bForceMonospaced=false
     bMaterialIsStencil=false
     fontMaterial=null
     fontObject=&HeadingNow
     letterSpacing=0
     monospacedWidth=1
     size=40.5
     skewAmount=0
     typefaceFontName=86 Bold Italic
     outlineSettings {
      bApplyOutlineToDropShadows=false
      bMiteredCorners=false
      bSeparateFillAlpha=false
      outlineMaterial=null
      outlineSize=0
      outlineColor=<rgba 0 0 0 1>
     }
    }
   }
   slot {class=OverlaySlot, properties {horizontalAlignment=HAlign_Center, verticalAlignment=VAlign_Center}}
  }
  PlayerIdentity_Pill {
   widgetClassPath=/PilgrimCore/Widgets/Player/WBP_Pilgrim_PlayerIdentity_Pill
   bIsVariable=true
   parent=PlayerName_Overlay
   widget=PlayerIdentity_Pill
   slot {class=OverlaySlot, properties {horizontalAlignment=HAlign_Center, verticalAlignment=VAlign_Bottom, padding=<ltrb 0 0 0 -40>}}
  }
 ]
}
